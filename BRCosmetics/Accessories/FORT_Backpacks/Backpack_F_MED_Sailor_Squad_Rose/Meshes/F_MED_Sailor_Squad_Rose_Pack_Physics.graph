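# Physics asset for the Sailor Squad Rose backpack cosmetic, giving the pack, headphones, straps and dangling charm figure physics bodies.
# The pack joins to root and carries the headphones, a four-box strap chain and the charm through charm_parent.
# The capsule charm branches from its body into head with ears, arms, gun, three-part tail and feet; the left foot joins the body with no leg body.
# Charm joints below charm_parent allow swing but zero twist.
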PhysicsAsset /BRCosmetics/Accessories/FORT_Backpacks/Backpack_F_MED_Sailor_Squad_Rose/Meshes/F_MED_Sailor_Squad_Rose_Pack_Physics
bodies {
 charm [{shapeType=Capsule, length=0.75, radius=1, shapeName=charm_capsule, center=<1.6737 0.5597 0.0769>, rotation=<rot 87.377 -179.9996 161.5272>}]
 charm_arm_l [{shapeType=Capsule, length=3, radius=1, shapeName=charm_arm_l_capsule, center=<2.7641 -0.0001 0.1324>, rotation=<rot -87.2568 0.0003 -0.0003>}]
 charm_arm_r [{
   shapeType=Capsule
   length=2.75
   radius=1
   shapeName=charm_arm_r_capsule
   center=<2.3508 -0.0001 0.1137>
   rotation=<rot -87.2309 0.0004 -0.0003>
  }]
 charm_body [{
   shapeType=Capsule
   length=1.25
   radius=2.25
   shapeName=charm_body_capsule
   center=<3.2122 0 0.836>
   rotation=<rot -87.7604 -179.9996 179.9995>
  }]
 charm_ear_l [{shapeType=Capsule, length=0.5, radius=1, shapeName=charm_ear_l_capsule, center=<0.7373 0 0>}]
 charm_ear_r [{shapeType=Capsule, length=0.5, radius=1, shapeName=charm_ear_r_capsule, center=<0.7373 0 0>}]
 charm_foot_l [{
   shapeType=Capsule
   length=2.1369
   radius=1.2408
   shapeName=charm_foot_l_capsule
   center=<1.1135 0.1149 0.058>
   rotation=<rot 0 91.6311 -89.6994>
  }]
 charm_foot_r [{shapeType=Capsule, length=2, radius=1, shapeName=charm_foot_r_capsule, center=<1.1572 0 -0.0299>, rotation=<rot -88.5117 -179.9992 179.9994>}]
 charm_gun [{shapeType=Capsule, length=4, radius=1, shapeName=charm_gun_capsule, center=<3.3591 0 -0.1264>, rotation=<rot 87.8442 0 0>}]
 charm_head [{shapeType=Capsule, length=1, radius=2.75, shapeName=charm_head_capsule, center=<-2.1555 -1.4655 -0.652>}]
 charm_leg_r [{shapeType=Capsule, length=2, radius=1, shapeName=charm_leg_r_capsule, center=<1.9547 0.0001 -0.0936>, rotation=<rot -88.8325 -0.0001 0.0001>}]
 charm_parent [{shapeType=Capsule, length=0.505, radius=0.505, rotation=<rot 0 -90 0>}]
 charm_tail_1 [{
   shapeType=Capsule
   length=1.5
   radius=1.25
   shapeName=charm_tail_1_capsule
   center=<3.8487 0.0728 0.0485>
   rotation=<rot -89.2897 -0.0147 2.413>
  }]
 charm_tail_2 [{
   shapeType=Capsule
   length=1
   radius=2.25
   shapeName=charm_tail_2_capsule
   center=<2.2639 0.0001 0.0274>
   rotation=<rot -89.3065 0.0012 -0.0012>
  }]
 charm_tail_3 [{shapeType=Capsule, length=0, radius=2, shapeName=charm_tail_3_capsule, center=<2.1281 0 0.0444>, rotation=<rot -88.8038 0.0002 -0.0001>}]
 headphones [
  {shapeType=Capsule, length=4.5, radius=4.25, center=<14.3355 -0.2926 0>}
  {shapeType=Capsule, length=3, radius=4.75, center=<11.6939 -1.303 -2.4688>, rotation=<rot -75.0003 -0.0002 0.0002>}
 ]
 pack [
  {shapeType=Capsule, length=24.75, radius=5.6919, center=<-0.0002 -4.6115 -8.8776>, rotation=<rot 90 0 0>}
  {shapeType=Capsule, length=10.75, radius=19.1919, center=<0 9.3161 -6.2221>, rotation=<rot 90 0 -7.9583>}
 ]
 root [{shapeType=Capsule, length=0.505, radius=0.505, rotation=<rot 0 -90 0>}]
 strap_1 [{shapeType=Box, extentX=2, extentY=0.5, extentZ=7.5, center=<3.2563 -0.1615 -0.0001>}]
 strap_2 [{shapeType=Box, extentX=2, extentY=0.5, extentZ=7.5, center=<3.2563 0 -0.0001>}]
 strap_3 [{shapeType=Box, extentX=4, extentY=0.5, extentZ=7.5, center=<1.7 0 0>}]
 strap_4 [{shapeType=Box, extentX=4, extentY=0.5, extentZ=7.5, center=<1.7 0 0>}]
}
constraints [
 {bone1Name=pack, bone2Name=root, swing1LimitDegrees=45, swing2LimitDegrees=45, twistLimitDegrees=45}
 {bone1Name=headphones, bone2Name=pack, swing1LimitDegrees=45, swing2LimitDegrees=45, twistLimitDegrees=5}
 {bone1Name=strap_1, bone2Name=pack, swing1LimitDegrees=30, swing2LimitDegrees=30, twistLimitDegrees=5}
 {bone1Name=strap_2, bone2Name=strap_1, swing1LimitDegrees=30, swing2LimitDegrees=30, twistLimitDegrees=5}
 {bone1Name=strap_3, bone2Name=strap_2, swing1LimitDegrees=45, swing2LimitDegrees=30, twistLimitDegrees=5}
 {bone1Name=strap_4, bone2Name=strap_3, swing1LimitDegrees=60, swing2LimitDegrees=30, twistLimitDegrees=5}
 {bone1Name=charm_parent, bone2Name=pack, swing1LimitDegrees=45, swing2LimitDegrees=45, twistLimitDegrees=45}
 {bone1Name=charm_arm_r, bone2Name=charm_body, swing1LimitDegrees=20, swing2LimitDegrees=30, twistLimitDegrees=0}
 {bone1Name=charm_gun, bone2Name=charm_arm_r, swing1LimitDegrees=30, swing2LimitDegrees=30, twistLimitDegrees=0}
 {bone1Name=charm_arm_l, bone2Name=charm_body, swing1LimitDegrees=30, swing2LimitDegrees=30, twistLimitDegrees=0}
 {bone1Name=charm_body, bone2Name=charm, swing1LimitDegrees=30, swing2LimitDegrees=30, twistLimitDegrees=0}
 {bone1Name=charm, bone2Name=charm_parent, swing1LimitDegrees=30, swing2LimitDegrees=30, twistLimitDegrees=0}
 {bone1Name=charm_leg_r, bone2Name=charm_body, swing1LimitDegrees=30, swing2LimitDegrees=30, twistLimitDegrees=0}
 {bone1Name=charm_tail_1, bone2Name=charm_body, swing1LimitDegrees=40, swing2LimitDegrees=40, twistLimitDegrees=0}
 {bone1Name=charm_tail_2, bone2Name=charm_tail_1, swing1LimitDegrees=40, swing2LimitDegrees=40, twistLimitDegrees=0}
 {bone1Name=charm_tail_3, bone2Name=charm_tail_2, swing1LimitDegrees=40, swing2LimitDegrees=40, twistLimitDegrees=0}
 {bone1Name=charm_head, bone2Name=charm_body, swing1LimitDegrees=15, swing2LimitDegrees=15, twistLimitDegrees=0}
 {bone1Name=charm_ear_l, bone2Name=charm_head, swing1LimitDegrees=30, swing2LimitDegrees=30, twistLimitDegrees=0}
 {bone1Name=charm_ear_r, bone2Name=charm_head, swing1LimitDegrees=30, swing2LimitDegrees=30, twistLimitDegrees=0}
 {bone1Name=charm_foot_r, bone2Name=charm_leg_r, swing1LimitDegrees=25, swing2LimitDegrees=25, twistLimitDegrees=0}
 {bone1Name=charm_foot_l, bone2Name=charm_body, swing1LimitDegrees=30, swing2LimitDegrees=30, twistLimitDegrees=0}
]
