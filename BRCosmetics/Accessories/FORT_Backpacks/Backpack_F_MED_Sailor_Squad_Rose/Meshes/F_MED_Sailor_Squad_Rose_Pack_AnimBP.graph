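# Animation Blueprint for the Sailor Squad Rose backpack, adding secondary physics motion to its hanging charm.
# The live chain runs the ref pose through RigidBody simulation relative to the pack bone, using the pack's physics asset.
# Bone modifiers then adjust charm head rotation and pack scale, before a translation constraint ties the left leg to pelvis and right foot.
# AnimDynamics nodes and crouch-weighted foot posing are unconnected to the output; the EventGraph is a stub.
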AnimBlueprint /BRCosmetics/Accessories/FORT_Backpacks/Backpack_F_MED_Sailor_Squad_Rose/Meshes/F_MED_Sailor_Squad_Rose_Pack_AnimBP
target_skeleton=&/F_MED_Sailor_Squad_Rose_Pack_Skeleton
graphs [
 AnimGraph {nodes [
   n1 = Anim Dynamics - Bone: charm_body(Alpha=0.7, ExternalForce="0,0,0") {
    angularSpringConstant=1500
    bGravityOverrideInSimSpace=true
    gravityScale=0.15
    linearSpringConstant=1000
    alphaBoolBlend {blendOption=Linear}
    boundBone {boneName=charm_body}
    componentAppliedLinearAccClamp=<6000 6000 6000>
    componentLinearAccScale=<0.15 0.15 0.2>
    componentLinearVelScale=<0.15 0.15 0.2>
    gravityOverride {z=-980}
    physicsBodyDefinitions [{
      sphereCollisionRadius=0
      boundBone {boneName=charm_body}
      boxExtents=<10 20 20>
      constraintSetup {angularLimitsMax=<5 5 30>, angularLimitsMin=<-5 -45 -60>, angularTarget=<0 0 0>}
      localJointOffset=<10 0 0>
     }]
    relativeSpaceBone {boneName=root}
   }
   n2 = Anim Dynamics - Bone: charm_head(ComponentPose=@n1, Alpha=0.7, ExternalForce="0,0,0") {
    angularSpringConstant=1500
    bGravityOverrideInSimSpace=true
    gravityScale=0.15
    linearSpringConstant=1000
    alphaBoolBlend {blendOption=Linear}
    boundBone {boneName=charm_head}
    componentAppliedLinearAccClamp=<6000 6000 6000>
    componentLinearAccScale=<0.15 0.2 0.2>
    componentLinearVelScale=<0.15 0.2 0.2>
    gravityOverride {z=-980}
    physicsBodyDefinitions [{
      sphereCollisionRadius=0
      boundBone {boneName=charm_head}
      boxExtents=<8 20 20>
      constraintSetup {angularLimitsMax=<30 5 30>, angularLimitsMin=<-80 -10 -5>, angularTarget=<0 0 0>}
      localJointOffset=<10 0 -10>
     }]
    relativeSpaceBone {boneName=pack}
   }
   n3 = Anim Dynamics - Bone: charm_ear_l(ComponentPose=@n2, Alpha=0.7, ExternalForce="0,0,0") {
    angularDampingOverride=0.9999
    angularSpringConstant=100
    bAngularSpring=true
    bGravityOverrideInSimSpace=true
    bOverrideAngularDamping=true
    simulationSpace=RootRelative
    alphaBoolBlend {blendOption=Linear}
    boundBone {boneName=charm_ear_l}
    componentAppliedLinearAccClamp=<3500 3500 3500>
    componentLinearAccScale=<1.2 1.2 1.2>
    componentLinearVelScale=<1.2 1.2 1.2>
    physicsBodyDefinitions [{
      sphereCollisionRadius=0
      boundBone {boneName=charm_ear_l}
      boxExtents=<20 10 10>
      constraintSetup {angularLimitsMax=<1 1 15>, angularLimitsMin=<-1 -30 -15>, angularTarget=<0 0 0>}
      localJointOffset=<2 0 0>
     }]
    relativeSpaceBone {boneName=pack}
   }
   Anim Dynamics - Bone: charm_ear_r(ComponentPose=@n3, Alpha=0.7, ExternalForce="0,0,0") {
    angularDampingOverride=0.9999
    angularSpringConstant=100
    bAngularSpring=true
    bGravityOverrideInSimSpace=true
    bOverrideAngularDamping=true
    simulationSpace=RootRelative
    alphaBoolBlend {blendOption=Linear}
    boundBone {boneName=charm_ear_r}
    componentAppliedLinearAccClamp=<3500 3500 3500>
    componentLinearAccScale=<1.2 1.2 1.2>
    componentLinearVelScale=<1.2 1.2 1.2>
    physicsBodyDefinitions [{
      sphereCollisionRadius=0
      boundBone {boneName=charm_ear_r}
      boxExtents=<20 10 10>
      constraintSetup {angularLimitsMax=<1 1 15>, angularLimitsMin=<-1 -30 -15>, angularTarget=<0 0 0>}
      localJointOffset=<2 0 0>
     }]
    relativeSpaceBone {boneName=pack}
   }
   Anim Dynamics - Bone: charm_foot_l(Alpha=1, ExternalForce="0,0,0") {
    angularDampingOverride=0.9999
    angularSpringConstant=100
    bAngularSpring=true
    bGravityOverrideInSimSpace=true
    bOverrideAngularDamping=true
    bUseGravityOverride=true
    bUsePlanarLimit=false
    simulationSpace=BoneRelative
    alphaBoolBlend {blendOption=Linear}
    boundBone {boneName=charm_foot_l}
    componentAppliedLinearAccClamp=<3500 3500 3500>
    componentLinearAccScale=<1.2 1.2 1.2>
    componentLinearVelScale=<1.2 1.2 1.2>
    gravityOverride {z=-980}
    physicsBodyDefinitions [{
      sphereCollisionRadius=0
      boundBone {boneName=charm_foot_l}
      boxExtents=<5 5 5>
      constraintSetup {angularLimitsMax=<25 20 15>, angularLimitsMin=<20 -10 -20>, angularTarget=<0 0 0>}
      localJointOffset=<4 -1 0>
     }]
    relativeSpaceBone {boneName=charm_leg_l}
   }
   Anim Dynamics - Bone: charm_leg_l(Alpha=1, ExternalForce="0,0,0") {
    angularDampingOverride=0.9999
    angularSpringConstant=100
    bAngularSpring=true
    bGravityOverrideInSimSpace=true
    bOverrideAngularDamping=true
    bUseGravityOverride=true
    simulationSpace=RootRelative
    alphaBoolBlend {blendOption=Linear}
    boundBone {boneName=charm_leg_l}
    componentAppliedLinearAccClamp=<3500 3500 3500>
    componentLinearAccScale=<1.1 1.1 1.1>
    componentLinearVelScale=<1.1 1.1 1.1>
    gravityOverride {z=-980}
    physicsBodyDefinitions [{
      sphereCollisionRadius=0
      boundBone {boneName=charm_leg_l}
      boxExtents=<20 25 25>
      constraintSetup {angularLimitsMax=<0 0 0>, angularLimitsMin=<0 0 0>, angularTarget=<0 0 0>}
      localJointOffset=<0 0 0>
     }]
    relativeSpaceBone {boneName=charm_pelvis}
   }
   Anim Dynamics - Bone: charm_leg_l(Alpha=1, ExternalForce="0,0,0") {
    alphaBoolBlend {blendOption=Linear}
    boundBone {boneName=charm_leg_l}
    physicsBodyDefinitions [{
      boundBone {boneName=charm_leg_l}
      boxExtents=<10 10 10>
      constraintSetup {angularLimitsMax=<0 0 0>, angularLimitsMin=<0 0 0>, angularTarget=<0 0 0>}
      localJointOffset=<0 0 0>
     }]
   }
   "Transform (Modify) Bone - Bone: charm_foot_l"(Alpha=Get Is Crouching() -> (Is Crouching:double), Translation="0,0,0", Rotation="69.370,-132.327,106.305", Scale="1.2,1.2,1.2") {
    rotationMode=BMM_Replace
    scaleSpace=BCS_BoneSpace
    alphaBoolBlend {blendOption=Linear}
    boneToModify {boneName=charm_foot_l}
    rotation=<rot 69.37 -132.327 106.305>
    scale=<1.2 1.2 1.2>
   }
   "Transform (Modify) Bone - Bone: charm_leg_l"(Alpha=1, Translation="0,0,0", Rotation="-35.975,-155.005,88.496", Scale="1,1,1") {
    rotationMode=BMM_Replace
    scaleSpace=BCS_BoneSpace
    alphaBoolBlend {blendOption=Linear}
    boneToModify {boneName=charm_leg_l}
    rotation=<rot -35.975 -155.005 88.496>
   }
   n4 = RigidBody(ComponentPose=Local To Component(LocalPose=Local Space Ref Pose())) {
    bClampLinearTranslationLimitToRefPose=true
    bOverrideWorldGravity=true
    overridePhysicsAsset=&/F_MED_Sailor_Squad_Rose_Pack_Physics
    simulationSpace=BaseBoneSpace
    alphaBoolBlend {blendOption=Linear}
    baseBoneRef {boneName=pack}
    componentAppliedLinearAccClamp=<1000 1000 1000>
    componentLinearAccScale=<0.3 0.3 0.4>
    componentLinearVelScale=<0.3 0.3 0.4>
    overrideWorldGravity {z=-980}
    simSpaceSettings {
     maxAngularAcceleration=0
     maxAngularVelocity=0.3
     maxLinearAcceleration=100
     maxLinearVelocity=100
     velocityScaleZ=0
     worldAlpha=0.5
     externalLinearDragV=<0.2 0.2>
    }
   }
   n5 = "Transform (Modify) Bone - Bone: charm_head"(ComponentPose=@n4, Alpha=1, Translation="0,0,0", Rotation="0.000,-40.000,0.000", Scale="1,1,1") {
    rotationMode=BMM_Additive
    scaleMode=BMM_Additive
    scaleSpace=BCS_BoneSpace
    alphaBoolBlend {blendOption=Linear}
    boneToModify {boneName=charm_head}
    rotation {yaw=-39.9998}
   }
   n6 = "Transform (Modify) Bone - Bone: pack"(ComponentPose=@n5, Translation="0,0,0", Rotation="0,0,0", Scale="1.2,1.2,1.2") {scaleMode=BMM_Additive, scaleSpace=BCS_BoneSpace, alphaBoolBlend {blendOption=Linear}, boneToModify {boneName=pack}, scale=<1.2 1.2 1.2>}
   n7 = "Constraint - charm_leg_l (TT)"!(ComponentPose=@n6, Alpha=1, ConstraintWeights_0=1, ConstraintWeights_1:float=1) -> (Pose) {
    alphaBoolBlend {blendOption=Linear}
    boneToModify {boneName=charm_leg_l}
    constraintSetup [{transformType=Translation, targetBone {boneName=charm_pelvis}}, {transformType=Translation, targetBone {boneName=charm_foot_r}}]
    constraintWeights [1,1]
   }
   Output Pose(Result=Component To Local(ComponentPose=@n7)) {layerGroup=None}
  ]}
 EventGraph {nodes [
   TryGetPawnOwner()
   Event BlueprintUpdateAnimation()
  ]}
]
variables [Is Crouching {type=number}]
